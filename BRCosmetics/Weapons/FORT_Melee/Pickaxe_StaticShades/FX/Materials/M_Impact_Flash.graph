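# Translucent unlit sprite material for the Static Shades pickaxe impact flash in Niagara effects.
# Particle color feeds emissive through MF_FX_EmissiveAdjustments; particle alpha times the texture's red channel gives opacity, with a parameter-blended depth fade off mobile.
# OpacityChannel selects among three lerp inputs that all receive red, so it changes nothing.
# A non-mobile vertex offset rescales sprites about the particle center by clamped camera distance, with zero offset on mobile.
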Material /BRCosmetics/Weapons/FORT_Melee/Pickaxe_StaticShades/FX/Materials/M_Impact_Flash
graph {
 properties {
  bUsedWithNanite=true
  bUsedWithNiagaraSprites=true
  blendMode=BLEND_Translucent
  pixelDepthOffsetMode=PDOM_Legacy
  shadingModel=MSM_Unlit
  usedShadingModels=Unlit
 }
 nodes [
  n1 = ParticleColor()
  MaterialFunctionCall(Emissive=@n1.RGB) {materialFunction=&MF_FX_EmissiveAdjustments}
  n2 = TextureSampleParameter2D() {parameterName=ImpactTexture, texture=/Game/Effects/Fort_Effects/Textures/Explo/T_Impact_Flash_2x2_01}
  n3 = Multiply(
   A=@n1.A
   B=MaterialFunctionCall!(
    1=@n2.R
    2=@n2.R
    3=@n2.R
    4
    A=DynamicParameter() {defaultValue=<rgba 0.5 0 0 0>, paramNames [OpacityChannel, Param2, Param3, Param4]}.OpacityChannel
   ) {materialFunction=/Engine/Functions/Engine_MaterialFunctions02/Lerp_Multiple_Float}.Lerp 3 Inputs
  )
  n4 = ScalarParameter() {defaultValue=200, parameterName=DepthFade}
  ShadingPathSwitch(
   Default=LinearInterpolate(A=@n3, B=DepthFade(Opacity=@n3, FadeDistance=@n4), Alpha=ScalarParameter() {parameterName=UseDepthFade})
   Mobile=@n3
  ) {desc=Migrated from FeatureLevelSwitch}
  n5 = ParticlePositionWS()
  n6 = Subtract(A=@n5, B=WorldPosition().XYZ)
  n7 = ScalarParameter() {defaultValue=5, parameterName=MaxSize}
  n8 = CameraPositionWS()
  n9 = Divide(
   A=Add(A=Distance(A=@n5, B=@n8), B=ScalarParameter() {defaultValue=-1250, parameterName=ScaleDistanceOffset})
   B=ScalarParameter() {defaultValue=20000, parameterName=Scale Distance}
  )
  ShadingPathSwitch(Default=Multiply(A=Multiply(A=@n6, B=@n7), B=Clamp(None=@n9)), Mobile=Constant())
 ]
}
