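# Niagara swing-trail effect for the Static Shades pickaxe, an electric-styled weapon trail with sparky pops.
# Trail_Faint and Trail_Sharp draw wide and narrow anim-trail ribbons that shrink and shift color over life.
# Trail_Lightning samples the user skeletal mesh for a flickering double-rendered ribbon with glow attributes, while Pops spawns jittering, flickering sprites in a sphere.
# Random-signed curves drive dynamic material parameters, and a time-of-day multiplier scales every emitter's alpha.
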NiagaraSystem /BRCosmetics/Weapons/FORT_Melee/Pickaxe_StaticShades/FX/NS_Pickaxe_StaticShades_Trail
dependencies {
 usedDataInterfaces [&NiagaraDataInterfaceCurve, &NiagaraDataInterfaceColorCurve, &NiagaraDataInterfaceArrayFloat]
 usedDynamicInputs [&FNTimeOfDayScalar]
 usedModules [
  &SystemState
  &self:SystemSpawnScript.SystemScriptSource.SystemScriptGraph.NiagaraNodeAssignment_0.SetVariables_#1
 ]
}
emitters {
 Pops {
  emitterUpdateScript [
   &EmitterState()
   &SpawnRate(Spawn Probability=0.75, SpawnRate=75, Use Spawn Probability=false)
  ]
  particleSpawnScript [
   &InitializeParticle(
    Color=<rgba 0 6 5.0637 1>
    Lifetime=&RandomRangeFloat(Minimum=0.166, Maximum=0.333)
    Position Offset=<0 0 -33>
    Position Offset Coordinate Space=::Local
    Sprite Rotation Mode=::NewEnumerator2
    Sprite Size Mode=::NewEnumerator1
    Uniform Sprite Size Max=5
    Uniform Sprite Size Min=2.5
    UsePositionOffset=true
    Write Material Random=true
   )
   &ShapeLocation(MaskSpawnGroup=false, Override Seed=false, Sphere Radius=11, UseOwnerScale=true)
   /SetVariables(Particles.SubImageIndex=&RandomRangeFloat(Minimum=0, Maximum=3)) {isSetParameters=true}
  ]
  particleUpdateScript [
   &ParticleState()
   &ScaleSpriteSize(
    Scale Factor=&MakeVector2D(
     X=&FloatFromCurve(
      FloatCurve=DataInterface() {propertyValues {Curve {keys [
          {arriveTangent=0.278, interpMode=RCIM_Cubic, leaveTangent=0.278, tangentMode=RCTM_User, time=0.4, value=1}
          {arriveTangent=4.8288, interpMode=RCIM_Cubic, leaveTangent=4.8288, tangentMode=RCTM_User, time=0.7, value=1.5}
         ]}}}
      Scale Curve=&RandomRangeFloat(Minimum=0.9, Maximum=1.1)
     )
     Y=&FloatFromCurve(
      FloatCurve=DataInterface() {propertyValues {Curve {keys [<at 0 1>]}}}
      Scale Curve=&RandomRangeFloat(Minimum=0.8, Maximum=1.2)
     )
    )
    Scale Sprite Size Mode=::NewEnumerator2
   )
   &DynamicMaterialParameters(
    Index 0 Param 1=&FloatFromCurve(FloatCurve=DataInterface() {propertyValues {Curve {keys [
         {arriveTangent=5.0926, interpMode=RCIM_Cubic, leaveTangent=5.0926, tangentMode=RCTM_User, time=0.6, value=0}
         {arriveTangent=-0.0012, interpMode=RCIM_Cubic, leaveTangent=-0.0012, tangentMode=RCTM_User, time=0.7, value=0.325}
         {arriveTangent=-0.0405, interpMode=RCIM_Cubic, leaveTangent=-0.0405, tangentMode=RCTM_User, time=1, value=0}
        ]}}})
    Index 0 Param 2=&FloatFromCurve(FloatCurve=DataInterface() {propertyValues {Curve {keys [
         {arriveTangent=0.1266, interpMode=RCIM_Cubic, leaveTangent=0.1266, tangentMode=RCTM_User, time=0.7, value=0.075}
         {arriveTangent=9.5183, interpMode=RCIM_Cubic, leaveTangent=9.5183, tangentMode=RCTM_User, time=0.95, value=1}
        ]}}})
    Index 0 Param 3=&FloatFromCurve(
     FloatCurve=DataInterface() {propertyValues {Curve {keys [<at 0 0.5 RCIM_Cubic>, {arriveTangent=3.2474, interpMode=RCIM_Cubic, leaveTangent=3.2474, tangentMode=RCTM_User, time=1, value=1.5}]}}}
     Scale Curve=&SelectFloatFromArray(
      Array Sampling Mode=::NewEnumerator0
      Direct Array Index=?
      Float Selection Array=DataInterface() {propertyValues {FloatData [-1,1]}}
      Random Seed=0
     )
    )
    Module.Param0WriteEnabled=true
    Module.Param1WriteEnabled=true
    Module.Param2WriteEnabled=true
    Module.Param3WriteEnabled=true
   )
   &Color(
    Color=&ColorFromCurve(ColorCurve=DataInterface() {propertyValues {
       AlphaCurve {keys [<at 0.093 0.05>,<at 0.1455 0.5>,<at 0.2996 0.03>,<at 0.3406 0.2>,<at 0.5095 0.03>,<at 1 0 RCIM_Cubic>]}
       BlueCurve {keys [<at 0.069 0>,<at 0.1695 0.5692>]}
       GreenCurve {keys [<at 0.069 13>,<at 0.1695 0>]}
       RedCurve {keys [<at 0.069 10.8333>,<at 0.1695 1>]}
      }})
    Scale Alpha=$System.TimeOfDayColorMultiplier
   )
   &JitterPosition()
  ]
  rendererClasses [&NiagaraSpriteRendererProperties]
  renderers [{rendererClass=&NiagaraSpriteRendererProperties}]
 }
 Trail_Faint {
  emitterUpdateScript [
   &EmitterState()
   &FN_SpawnAnimTrail_v2(AnimTrailWidthMode=::FromSecond)
  ]
  particleSpawnScript [
   /Game/Effects/Niagara/Modules/AnimTrail/FN_InitializeAnimTrail/FN_InitializeAnimTrailBoneIndecies(Second Bone or Socket Index=1)
   /Game/Effects/Niagara/Modules/AnimTrail/FN_InitializeAnimTrail/FN_InitializeAnimTrail(AnimTrailWidth=1, Color=<rgba 0.4526 0 0.9734 0.2>, Lifetime=0.3)
   &SystemLocation(MaskSpawnGroup=false)
  ]
  particleUpdateScript [
   &ScaleRibbonWidth(Ribbon Width Scale=&FloatFromCurve(FloatCurve=DataInterface() {propertyValues {Curve {keys [
         {arriveTangent=-1.2579, interpMode=RCIM_Cubic, time=0, value=1}
         {arriveTangent=-0.9664, interpMode=RCIM_Cubic, leaveTangent=-0.9664, tangentMode=RCTM_User, time=1, value=0}
        ]}}}))
   &FN_UpdateAnimTrail_v2()
   &Color(
    Color=&ColorFromCurve(ColorCurve=DataInterface() {propertyValues {
       AlphaCurve {keys [<at 0 0.2 RCIM_Cubic>,<at 1 0 RCIM_Cubic>]}
       BlueCurve {keys [<at 0.069 0.4558>, {interpMode=RCIM_Cubic, leaveTangent=1.0039, time=0.3954, value=1}]}
       GreenCurve {keys [<at 0.069 1>, {interpMode=RCIM_Cubic, leaveTangent=-0.727, time=0.3954, value=0}]}
       RedCurve {keys [<at 0.069 0>,<at 0.3954 0.3926 RCIM_Cubic>]}
      }})
    Scale Alpha=$System.TimeOfDayColorMultiplier
   )
  ]
  rendererClasses [&NiagaraRibbonRendererProperties]
  renderers [{rendererClass=&NiagaraRibbonRendererProperties}]
 }
 Trail_Lightning {
  emitterUpdateScript [
   &EmitterState()
   &SpawnRate(Spawn Probability=0.5, SpawnRate=150, Use Spawn Probability=false)
  ]
  particleSpawnScript [
   &InitializeParticle(
    Color=<rgba 0.8 0.8 0.8 1>
    Lifetime=0.222
    Ribbon Width=20
    Ribbon Width Mode=::NewEnumerator1
    UsePositionOffset=false
    Write Material Random=true
   )
   &SkeletalMeshLocation(
    Bone / Socket Index=1
    MaskSpawnGroup=false
    Mesh Sampling Type=::NewEnumerator5
    Skeletal Mesh=$User.SkeletalMesh
    Socket Sampling Mode=::NewEnumerator5
   )
  ]
  particleUpdateScript [
   &ParticleState()
   &Color(
    Color=&ColorFromCurve(ColorCurve=DataInterface() {propertyValues {
       AlphaCurve {keys [<at 0 2>,<at 0.1284 0.05>,<at 0.2996 0.03>,<at 0.3406 0.4>,<at 0.5095 0.03>,<at 1 0 RCIM_Cubic>]}
       BlueCurve {keys [<at 0.069 0>,<at 0.1695 0.5692>]}
       GreenCurve {keys [<at 0.069 6>,<at 0.1695 0>]}
       RedCurve {keys [<at 0.069 5>,<at 0.1695 1>]}
      }})
    Scale Alpha=$System.TimeOfDayColorMultiplier
   )
   /SetVariables(Particles.Glow Color=&MultiplyLinearColor(
     Linear Color=$Particles.Color
     Scale Factor=&MakeFloat4FromVectorAndFloat("Vector (RGB)"=<1 1 1>, "Float (Alpha)"=0.3)
    )) {isSetParameters=true}
   /SetVariables(Particles.Glow Width:NiagaraFloat=&Multiply_Float(A=$Particles.RibbonWidth, B=1.5)) {isSetParameters=true}
   &DynamicMaterialParameters(
    Index 0 Param 2=&FloatFromCurve(
     FloatCurve=DataInterface() {propertyValues {Curve {keys [
         {arriveTangent=0.9907, interpMode=RCIM_Cubic, leaveTangent=0.9907, tangentMode=RCTM_User, time=0.5, value=0}
         {arriveTangent=8.2787, interpMode=RCIM_Cubic, leaveTangent=8.2787, tangentMode=RCTM_User, time=1, value=1.625}
        ]}}}
     Scale Curve=&Multiply_Float(
      A=&RandomRangeFloat(Minimum=0.003, Maximum=0.007)
      B=&SelectFloatFromArray(
       Array Sampling Mode=::NewEnumerator0
       Direct Array Index=?
       Float Selection Array=DataInterface() {propertyValues {FloatData [-3,3]}}
       Random Seed=0
      )
     )
    )
    Module.Param0WriteEnabled=true
    Module.Param1WriteEnabled=true
    Module.Param2WriteEnabled=true
    Module.Param3WriteEnabled=true
   )
  ]
  rendererClasses [&NiagaraRibbonRendererProperties]
  renderers [
   {rendererClass=&NiagaraRibbonRendererProperties}
   {rendererClass=&NiagaraRibbonRendererProperties, rendererIndex=1}
  ]
 }
 Trail_Sharp {
  emitterUpdateScript [
   &EmitterState()
   &FN_SpawnAnimTrail_v2(AnimTrailWidthMode=::FromSecond)
  ]
  particleSpawnScript [
   &FN_InitializeAnimTrailBoneIndecies(Second Bone or Socket Index=1)
   &FN_InitializeAnimTrail(AnimTrailWidth=0.3, Color=<rgba 0.0823 0.7682 0.4969 2>, Lifetime=0.2)
   &SystemLocation(MaskSpawnGroup=false)
  ]
  particleUpdateScript [
   &ScaleRibbonWidth(Ribbon Width Scale=&FloatFromCurve(FloatCurve=DataInterface() {propertyValues {Curve {keys [
         {arriveTangent=-2.7553, interpMode=RCIM_Cubic, leaveTangent=-2.7553, tangentMode=RCTM_User, time=0, value=1}
         {arriveTangent=-0.1336, interpMode=RCIM_Cubic, leaveTangent=-0.1336, tangentMode=RCTM_User, time=1, value=0}
        ]}}}))
   &FN_UpdateAnimTrail_v2()
   &Color(
    Color=&ColorFromCurve(ColorCurve=DataInterface() {propertyValues {
       AlphaCurve {keys [<at 0.2961 0.5>,<at 1 0 RCIM_Cubic>]}
       BlueCurve {keys [<at 0 0.4259 RCIM_Cubic>,<at 0.069 0>,<at 0.222 0>,<at 0.287 0.5298>,<at 0.3863 1>]}
       GreenCurve {keys [<at 0 1 RCIM_Cubic>,<at 0.069 3>,<at 0.222 3>,<at 0.287 2>,<at 0.3863 0>]}
       RedCurve {keys [<at 0 0 RCIM_Cubic>,<at 0.069 2.35>,<at 0.222 2.35>,<at 0.287 0>,<at 0.3863 0.5903>]}
      }})
    Scale Alpha=$System.TimeOfDayColorMultiplier
   )
  ]
  rendererClasses [&NiagaraRibbonRendererProperties]
  renderers [{rendererClass=&NiagaraRibbonRendererProperties}]
 }
}
summary {
 systemName=NS_Pickaxe_StaticShades_Trail
 emitters [
  {emitterName=Trail_Faint, rendererClasses [&NiagaraRibbonRendererProperties]}
  {emitterName=Trail_Lightning, rendererClasses [&NiagaraRibbonRendererProperties]}
  {emitterName=Trail_Sharp, rendererClasses [&NiagaraRibbonRendererProperties]}
  {emitterName=Pops, rendererClasses [&NiagaraSpriteRendererProperties]}
 ]
 userVariables [
  User.RibbonIDOffset {type=NiagaraInt32}
  User.SkeletalMesh {
   type=NiagaraDataInterfaceSkeletalMesh
   defaultValue {
    struct=NiagaraExt_VariableValue_DataInterface
    value {dataInterface=&self:NiagaraDataInterfaceSkeletalMesh_0}
   }
  }
 ]
}
